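; Gameplay ability for the Collect All Seeds quest in BattlePassS21, applying a timer-tag gameplay effect when activated from an event.
; After calling its parent, it resolves the player controller from the owning player pawn.
; If that controller is valid, it applies GE_Quest_CollectAllSeeds_TimerTag to the controller, then ends the ability.
; The invalid-controller branch is unhandled, so this graph does not end the ability in that case.
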(event Ability|EventActivateAbilityFromEvent
  (params (EventData :type Gameplay%20Event%20Data%20Structure%20%28by%20ref%29))
  (bind _owningplayerpawn (Variables|GAVars|Data|GetOwningPlayerPawn))
  (bind _returnvalue (Fortnite|GetFortPlayerControllerfromActor _owningplayerpawn))
  (|CallParentFunction EventData)
  (Utilities|IsValid _returnvalue
    (:Is%20Valid
      (Creative|GameplayEffects|SimpleApplyGameplayEffectToActor "/BattlePassS21/Items/QuestAbilities/Quest_CollectAllSeeds/GE_Quest_CollectAllSeeds_TimerTag.GE_Quest_CollectAllSeeds_TimerTag_C" _returnvalue)
      (Ability|EndAbility))))
