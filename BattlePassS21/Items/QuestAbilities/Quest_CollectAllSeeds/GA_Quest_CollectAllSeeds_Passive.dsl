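; Passive gameplay ability that reports progress for the "Collect All Seeds" quest in Battle Pass S21.
; On activation, it calls the parent logic and waits for a gameplay event matching its Event Tag variable.
; Each received event sends a custom stat event tagged Athena.Quests.CollectAllSeeds, with a count of one, to the owning player's quest manager.
; The end and failed passive activation handlers both call End Ability.
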(event Ability|EventActivateAbility
  (|CallParentFunction)
  (Ability|Tasks|WaitGameplayEvent (Variables|Default|GetEvent%20Tag)
    (:EventReceived
      (Quest|SendComplexCustomStatEvent (Variables|GAVars|Data|GetOwningPlayerQuestManager) 0 "(GameplayTags=((TagName=\"Athena.Quests.CollectAllSeeds\")))" "(GameplayTags=)" 1))))

(event Ability|EventOnEndAbility
  (params (bWasCancelled :type Boolean))
  (Ability|EndAbility))

(event Ability|EventFailedToActivatePassiveAbility
  (params (ActorInfo :type Gameplay%20Ability%20Actor%20Info%20Structure))
  (Ability|EndAbility))
